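# Landscape material function that outputs Season 18 Corruption masks: island-wide, dynamic encounter, combined, noise-biased and second-stage.
# It samples a render target through world-projected UVs for island progression, with optional explicit mip level.
# A Sideways bubble sphere mask, optionally evaluated at the pivot, is maxed in except for RVT use, then noise-biased.
# Player-two (disabled for split-screen performance) and second-stage masks stay out of the combined result.
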MaterialFunction /Game/Athena/Apollo/Environments/Landscape/Materials/Corruption/MF_Corruption_LandscapeMask
graph {nodes [
  Constant() {r=0.5}
  n1 = VectorParameter() {group=S18_Corruption, parameterName=CorruptionProjectionBias, defaultValue=<rgba 30000.125 30000.125 0 0>}
  n2 = Add(A=WorldPosition().XYZ, B=@n1.RGB)
  n3 = VectorParameter() {group=S18_Corruption, parameterName=CorruptionProjectionScale, defaultValue=<rgba 60000.25 60000.25 0 0>}
  n4 = CollectionParameter() {
   collection=&FortniteMaterialParameters
   parameterName=CorruptionProjectionBias
  }
  n5 = Add(A=AppendVector(A=WorldPosition().XYZ, B=Constant()), B=@n4)
  n6 = CollectionParameter() {
   collection=&FortniteMaterialParameters
   parameterName=CorruptionProjectionScale
  }
  n7 = ComponentMask(None=StaticSwitchParameter(True=Divide(A=@n2, B=@n3.RGB), False=Divide(A=@n5, B=@n6)) {group=S18_Corruption, parameterName=SetCorruptionBiasScaleManually}) {g=true, r=true}
  n8 = TextureSampleParameter2D!(
   UVs=@n7
   Level=FunctionInput(Preview=Constant()) {bUsePreviewValueAsDefault=true, inputName=RTMipLevel, inputType=FunctionInput_Scalar, sortPriority=6}
   Apply View MipBias
  ) {
   group=S18_Corruption
   mipValueMode=TMVM_MipLevel
   parameterName=CorruptionRenderTarget
   samplerType=SAMPLERTYPE_LinearColor
   texture=&/RT_CorruptionProcessed
  }
  n9 = TextureSampleParameter2D!(UVs=@n7, Apply View MipBias) {
   group=S18_Corruption
   parameterName=CorruptionRenderTarget
   samplerType=SAMPLERTYPE_LinearColor
   texture=&RT_CorruptionProcessed
  }
  n10 = StaticSwitch(
   True=@n8.R
   False=@n9.R
   Value=FunctionInput(Preview=StaticBool()) {bUsePreviewValueAsDefault=true, inputName=SpecifyRTMipLevel, inputType=FunctionInput_StaticBool, sortPriority=5}
  )
  n11 = FunctionOutput(None=@n10) {outputName=IslandWideProgressionMask, sortPriority=2}
  n12 = CollectionParameter() {
   collection=&FortniteMaterialParameters
   parameterName=SidewaysBubble_Location_LocalP1
  }
  n13 = FunctionInput(Preview=WorldPosition() {worldPositionShaderOffset=WPT_ExcludeAllShaderOffsets}.XYZ) {bUsePreviewValueAsDefault=true, inputName=WorldPosition, sortPriority=3}
  n14 = FunctionInput() {bUsePreviewValueAsDefault=true, inputName=PivotPoint, sortPriority=2}
  n15 = FunctionInput(Preview=StaticBool()) {
   bUsePreviewValueAsDefault=true
   inputName=TransformPositionToPivot
   inputType=FunctionInput_StaticBool
   sortPriority=1
   previewValue=<0 0 0 0>
  }
  n16 = StaticSwitch(True=Add(A=@n13, B=Subtract(A=@n14, B=@n13) {desc=Find delta}) {desc=offset}, False=@n13, Value=@n15)
  n17 = CollectionParameter() {
   collection=&FortniteMaterialParameters
   parameterName=SidewaysBubble_Scale_LocalP1
  }
  n18 = ShadingPathSwitch(Default=Constant() {r=0.75}, Mobile=Constant() {r=0.9})
  n19 = SphereMask(A=ComponentMask(None=@n12) {b=true, g=true, r=true}, B=@n16, Radius=@n17, Hardness=@n18)
  n20 = FunctionOutput(None=@n19) {outputName=DyamicEncountersMask, sortPriority=1}
  n21 = StaticBool()
  n22 = FunctionOutput(None=StaticSwitch(
    True=@n11
    False=Max(A=@n11, B=@n20)
    Value=FunctionInput(Preview=@n21) {bUsePreviewValueAsDefault=true, inputName=UsedWithRVT, inputType=FunctionInput_StaticBool, sortPriority=4}
   )) {outputName=CombinedMasks, sortPriority=3}
  n23 = FunctionInput() {bUsePreviewValueAsDefault=true, inputName=NoiseMask, inputType=FunctionInput_Scalar, previewValue=<1 0 0 1>}
  n24 = Add(A=@n22, B=Multiply(A=Ceil(None=@n22), B=Subtract(A=@n23)))
  FunctionOutput(None=Saturate(None=Add(A=@n22, B=@n24))) {outputName=CombinedMasks_NoiseBiased}
  n25 = CollectionParameter() {
   collection=&FortniteMaterialParameters
   parameterName=SidewaysBubble_Location_LocalP2
  }
  n26 = CollectionParameter() {
   collection=&FortniteMaterialParameters
   parameterName=SidewaysBubble_Scale_LocalP2
  }
  Max(A=@n19, B=SphereMask(A=ComponentMask(None=@n25) {b=true, g=true, r=true}, B=@n16, Radius=@n26, Hardness=@n18)) {desc=Disabling split screen due to perf issues}
  Reroute(None=MaterialFunctionCall!(Time, Divisor=Constant() {r=0}, Mod=Constant() {r=225000}) {materialFunction=/Game/Effects/Fort_Effects/Materials/_Functions/MF_ModuloTime}.Unclamped)
  n27 = CollectionParameter() {collection=&FortniteMaterialParameters, parameterName=Corruption2ndStage_0-1}
  n28 = SphereMask(
   A=Constant3Vector() {constant=<rgba 5120 7168 0 0>}
   B=Multiply(A=WorldPosition() {worldPositionShaderOffset=WPT_ExcludeAllShaderOffsets}.XYZ, B=Constant3Vector() {constant=<rgba 1 1 0 0>})
   Radius=@n27
   Hardness=Constant() {r=0.95}
  )
  Saturate(None=Max(A=@n10, B=FunctionOutput(None=@n28) {outputName=Corruption2ndStage, sortPriority=4})) {desc=Reconnect later in the season when we need the 2nd stage}
 ]}
