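# Material function that generates world-planar UV sets for the Apollo landscape corruption material.
# World position, excluding shader offsets by default, is masked to XY, output, and divided by separate scalar parameters for bubble, corruption, chroma, distortion and lerp-mask UVs.
# A TransformPositionToPivot static switch adds the pivot-minus-position delta to the position, which algebraically resolves to the pivot point itself.
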MaterialFunction /Game/Athena/Apollo/Environments/Landscape/Materials/Corruption/MF_Corruption_Landscape_UVs
graph {nodes [
  FunctionOutput(None=Divide(A=NamedRerouteUsage(), B=ScalarParameter() {defaultValue=3450, desc=3600, parameterName=BubblesDivisor})) {outputName=BubblesUVs}
  FunctionOutput(None=Divide(A=NamedRerouteUsage(), B=ScalarParameter() {defaultValue=386.25, parameterName=CorruptionDivisor})) {outputName=CorruptionUVs, sortPriority=1}
  FunctionOutput(None=Divide(A=NamedRerouteUsage(), B=ScalarParameter() {defaultValue=400, parameterName=ChromaDivisor})) {outputName=ChromaUVs, sortPriority=2}
  FunctionOutput(None=Divide(A=NamedRerouteUsage(), B=ScalarParameter() {defaultValue=1200, parameterName=DistortionTileDivisor})) {outputName=DistortionUVs, sortPriority=3}
  n1 = FunctionInput(Preview=WorldPosition() {worldPositionShaderOffset=WPT_ExcludeAllShaderOffsets}.XYZ) {bUsePreviewValueAsDefault=true, inputName=WorldPosition, sortPriority=2}
  n2 = FunctionInput() {bUsePreviewValueAsDefault=true, inputName=PivotPoint, sortPriority=1, previewValue=<0 0 0 0>}
  n3 = FunctionInput(Preview=StaticBool()) {bUsePreviewValueAsDefault=true, inputName=TransformPositionToPivot, inputType=FunctionInput_StaticBool, previewValue=<0 0 0 0>}
  n4 = StaticSwitch(True=Add(A=@n1, B=Subtract(A=@n2, B=@n1) {desc=Find delta}) {desc=offset}, False=@n1, Value=@n3)
  FunctionOutput(None=NamedRerouteDeclaration(None=ComponentMask(None=@n4) {g=true, r=true}) {name=WorldPos.XY, nodeColor=<rgba 0.4766 1 0 1>}) {outputName=WorldPos.XY, sortPriority=-1}
  FunctionOutput(None=Divide(A=NamedRerouteUsage(), B=ScalarParameter() {defaultValue=-824, parameterName=CorruptionB_Divisor})) {outputName=CorruptionB_UVs, sortPriority=4}
  FunctionOutput(None=Divide(A=NamedRerouteUsage(), B=ScalarParameter() {defaultValue=6480, parameterName=CorruptionHQ_LERP_Divisor})) {outputName=Corruption_HQ_Lerp, sortPriority=5}
  FunctionOutput(None=Divide(A=NamedRerouteUsage(), B=ScalarParameter() {defaultValue=5472, desc=5471, parameterName=BubblesBDivisor})) {outputName=BubblesB_UVs, sortPriority=6}
  FunctionOutput(None=Divide(A=NamedRerouteUsage(), B=ScalarParameter() {defaultValue=15000, parameterName=BubblesLerpDivisor})) {outputName=BubblesLerpMask, sortPriority=7}
 ]}
