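# Mobile landscape material function that applies corruption to base color in the Apollo environment.
# It desaturates incoming base color and tints it with a corruption color times snow-texture noise, which fades out with camera distance.
# That noise also feeds the corruption mask function, whose mask times a cliff slope input drives the blend.
# It outputs updated attributes and the color separately; the second-stage color blend is left unconnected.
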MaterialFunction /Game/Athena/Apollo/Environments/Landscape/Materials/Corruption/MF_Corruption_Landscape_Mobile
graph {nodes [
  n1 = GetMaterialAttributes(None=FunctionInput() {inputName=MA_In, inputType=FunctionInput_MaterialAttributes}) {attributeGetTypes [#1]}
  n2 = ScalarParameter() {defaultValue=0.7, parameterName=CorruptionMobileDesaturation}
  n3 = VectorParameter() {parameterName=CorruptionColorMobile, defaultValue=<rgba 2 0.628 0.04 0>}
  n4 = TextureSample(UVs=MaterialFunctionCall!(TransformPositionToPivot, PivotPoint, WorldPosition) {materialFunction=&/MF_Corruption_Landscape_UVs}.CorruptionUVs) {samplerType=SAMPLERTYPE_LinearColor, texture=/Game/Athena/Environments/Landscape/Textures/T_Athena_Snow_01_MOBILE_M}
  n5 = Constant() {r=2000}
  n6 = Constant() {r=8000}
  n7 = Distance(A=CameraPositionWS(), B=WorldPosition() {worldPositionShaderOffset=WPT_ExcludeAllShaderOffsets}.XYZ)
  n8 = LinearInterpolate(A=@n4.B, Alpha=Saturate(None=InverseLinearInterpolate(A=@n5, B=@n6, Value=@n7)))
  n9 = MaterialFunctionCall!(NoiseMask=@n4.B, TransformPositionToPivot, PivotPoint, WorldPosition, UsedWithRVT, SpecifyRTMipLevel, RTMipLevel) {materialFunction=&/MF_Corruption_LandscapeMask}
  n10 = FunctionInput(Preview=DotProduct(A=VertexNormalWS(), B=Constant3Vector() {constant=<rgba 0 0 1 0>})) {bUsePreviewValueAsDefault=true, inputName=CliffSlopeMask, inputType=FunctionInput_Scalar, previewValue=<1 0 0 1>}
  n11 = LinearInterpolate(
   A=@n1.BaseColor
   B=Multiply(A=Desaturation(None=@n1.BaseColor, Fraction=@n2), B=Multiply(A=@n3.RGB, B=@n8))
   Alpha=Multiply(A=@n9.CombinedMasks, B=@n10)
  )
  FunctionOutput(None=SetMaterialAttributes!(MaterialAttributes=@n1.MaterialAttributes, BaseColor=@n11) {attributeSetTypes [#1]})
  FunctionOutput(None=@n11) {outputName=BaseColorMobile, sortPriority=1}
  LinearInterpolate(
   A=@n3.RGB
   B=VectorParameter() {parameterName=Corruption2ndStageColorMobile, defaultValue=<rgba 0.651 0.175 0.6788 0>}.RGB
   Alpha=@n9.Corruption2ndStage
  )
 ]}
