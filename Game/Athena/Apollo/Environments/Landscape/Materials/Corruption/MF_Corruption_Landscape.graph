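# Landscape material function that layers animated, iridescent corruption over incoming material attributes, also outputting masks, base color and isolated attributes.
# World-position UVs sample swirl, bubble and normal textures, driving curve-atlas color LUTs, with a collection-driven second-stage tint and sine-banded dark areas altering color and surface response.
# A noise-based mask with optional height fade controls coverage.
# Quality switches reserve extra detail for higher tiers, while mobile shading uses a separate function.
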MaterialFunction /Game/Athena/Apollo/Environments/Landscape/Materials/Corruption/MF_Corruption_Landscape
graph {nodes [
  n1 = NamedRerouteUsage()
  n2 = NamedRerouteUsage()
  n3 = Constant() {r=0.045}
  n4 = SetMaterialAttributes!(MaterialAttributes, BaseColor=@n2, Normal=NamedRerouteUsage(), Metallic=NamedRerouteUsage(), Specular=@n3, Roughness=NamedRerouteUsage()) {attributeSetTypes [
    #1
    #2
    #3
    #4
    #5
   ]}
  n5 = SetMaterialAttributes!(MaterialAttributes, BaseColor=@n2, Specular=@n3, Normal=NamedRerouteUsage(), Metallic=NamedRerouteUsage(), Roughness=NamedRerouteUsage()) {attributeSetTypes [
    #1
    #4
    #2
    #3
    #5
   ]}
  n6 = FunctionOutput(None=QualitySwitch(Default=@n4, Low=@n5, Medium=@n5)) {outputName=CorruptionMA_Isolated, sortPriority=2}
  n7 = NamedRerouteUsage()
  n8 = MaterialFunctionCall!(CliffSlopeMask, MA_In=NamedRerouteUsage()) {materialFunction=&/MF_Corruption_Landscape_Mobile}
  FunctionOutput(None=ShadingPathSwitch(Default=BlendMaterialAttributes(A=@n1, B=@n6, Alpha=@n7) {vertexAttributeBlendType=UseA}, Mobile=@n8.Result))
  n9 = Subtract(A=NamedRerouteUsage()) {constB=0.5}
  n10 = NamedRerouteUsage()
  n11 = Subtract(A=ComponentMask(None=@n10) {a=true}) {constB=2, desc="0.5 originally, now 0.5 + 1.5"}
  n12 = NamedRerouteUsage()
  n13 = AppendVector(A=Constant() {desc=WorldOffsetScaleX, r=0.0001}, B=Constant() {desc=WorldOffsetScaleY, r=0.0001})
  n14 = Add(
   A=Multiply(A=Add(A=@n9, B=@n11), B=Constant() {desc="DarkAreaDetailScale ", r=0.15})
   B=Add(A=ComponentMask(None=@n10) {b=true}, B=DotProduct(A=@n12, B=@n13))
  )
  n15 = NamedRerouteDeclaration(None=ConstantBiasScale(None=Sine(None=@n14) {period=1.5})) {name=DarkAreasMask, nodeColor=<rgba 0.5937 1 0 1>}
  FunctionOutput(None=@n15) {outputName=Dark Areas, sortPriority=1}
  FunctionOutput(None=ShadingPathSwitch(Default=NamedRerouteUsage(), Mobile=@n8.BaseColorMobile)) {outputName=BaseColor, sortPriority=3}
  n16 = MaterialFunctionCall!(
   NoiseMask=NamedRerouteUsage()
   TransformPositionToPivot
   PivotPoint
   WorldPosition=NamedRerouteUsage()
   UsedWithRVT=FunctionInput(Preview=StaticBool()) {bUsePreviewValueAsDefault=true, inputName=UsedWithRVT, inputType=FunctionInput_StaticBool, sortPriority=5}
   SpecifyRTMipLevel
   RTMipLevel
  ) {materialFunction=&/MF_Corruption_LandscapeMask}
  FunctionOutput(None=@n16.CombinedMasks) {outputName=CombinedMasks, sortPriority=4}
  n17 = ScalarParameter() {defaultValue=0.2, parameterName=Metallic}
  n18 = ScalarParameter() {defaultValue=0.5, parameterName=MetallicDarkAreas}
  n19 = NamedRerouteUsage()
  FunctionOutput(None=NamedRerouteDeclaration(None=LinearInterpolate(A=@n17, B=@n18, Alpha=@n19)) {name=MetallicOut, nodeColor=<rgba 0 0.5547 1 1>}) {outputName=MetallicValues, sortPriority=5}
  FunctionOutput(None=@n16.Corruption2ndStage) {outputName=CorruptionSecondStage, sortPriority=6}
  n20 = GetMaterialAttributes(None=FunctionInput() {inputName=MA_In, inputType=FunctionInput_MaterialAttributes}) {attributeGetTypes [
    #1
    #2
    #6
    #3
    #5
   ]}
  NamedRerouteDeclaration(None=@n20.BaseColor) {name=BaseColorIn, nodeColor=<rgba 0.0078 0 1 1>}
  n21 = TextureSample(UVs=NamedRerouteUsage()) {
   samplerSource=SSM_Wrap_WorldGroupSettings
   samplerType=SAMPLERTYPE_LinearColor
   texture=/Game/Global/Textures/Default/Tileable_Patterns/T_Corruption_SwirlPack
  }
  n22 = TextureSample(UVs=NamedRerouteUsage()) {
   samplerSource=SSM_Wrap_WorldGroupSettings
   samplerType=SAMPLERTYPE_LinearColor
   texture=&T_Corruption_SwirlPack
  }
  n23 = TextureSample(UVs=NamedRerouteUsage()) {
   samplerSource=SSM_Wrap_WorldGroupSettings
   samplerType=SAMPLERTYPE_LinearColor
   texture=&T_Corruption_SwirlPack
  }
  n24 = QualitySwitch(
   Default=@n21.RGB
   Epic=LinearInterpolate(A=@n21.RGB, B=@n22.RGB, Alpha=NamedRerouteDeclaration(None=@n23.A) {name=TextureBreakupMask, nodeColor=<rgba 1 0.3984 0 1>}) {constAlpha=1}
  )
  NamedRerouteDeclaration(None=ComponentMask(None=@n24) {r=true}) {name=Swirls, nodeColor=<rgba 0 0.2734 1 1>}
  NamedRerouteDeclaration(None=@n20.Roughness) {name=RoughnessIn, nodeColor=<rgba 0 0.7656 1 1>}
  n25 = Distance(A=WorldPosition().XYZ, B=CameraPositionWS())
  n26 = ScalarParameter() {defaultValue=10000, parameterName=CorruptionDistanceFade_Epic}
  NamedRerouteDeclaration(None=Saturate(None=Divide(A=@n25, B=@n26))) {name=DistanceFade, nodeColor=<rgba 0 0.0859 1 1>}
  n27 = WorldPosition()
  n28 = MaterialFunctionCall!(
   TransformPositionToPivot
   PivotPoint
   WorldPosition=NamedRerouteDeclaration(None=FunctionInput(Preview=@n27.XYZ) {bUsePreviewValueAsDefault=true, inputName=WorldPosition, sortPriority=3}) {name=WorldPositionInput, nodeColor=<rgba 1 0.5391 0 1>}
  ) {materialFunction=&/MF_Corruption_Landscape_UVs}
  NamedRerouteDeclaration(None=@n28.CorruptionB_UVs) {name=CorruptionB_UVs, nodeColor=<rgba 1 0 0.5625 1>}
  NamedRerouteDeclaration(None=LinearInterpolate(
    A=ScalarParameter() {defaultValue=0.9, parameterName=Roughness}
    B=ScalarParameter() {defaultValue=0.28, parameterName=RoughnessDarkAreas}
    Alpha=NamedRerouteUsage()
   )) {name=RoughnessOut, nodeColor=<rgba 0.1484 1 0 1>}
  NamedRerouteDeclaration(None=@n28.WorldPos.XY) {name=WorldPosition.XY, nodeColor=<rgba 0 1 0.5547 1>}
  NamedRerouteDeclaration(None=@n28.Corruption_HQ_Lerp) {name=Corruption_HQ_Lerp_UVs, nodeColor=<rgba 0 0.5312 1 1>}
  n29 = TextureSample(UVs=NamedRerouteUsage()) {
   samplerSource=SSM_Wrap_WorldGroupSettings
   samplerType=SAMPLERTYPE_Normal
   texture=&/T_Corruption_Normals_Mixed
  }
  n30 = TextureSample(UVs=NamedRerouteUsage()) {
   samplerSource=SSM_Wrap_WorldGroupSettings
   samplerType=SAMPLERTYPE_Normal
   texture=&T_Corruption_Normals_Mixed
  }
  n31 = LinearInterpolate(A=@n29.RGB, B=@n30.RGB, Alpha=NamedRerouteUsage())
  NamedRerouteDeclaration(None=QualitySwitch(Default=@n29.RGB, Epic=Normalize(VectorInput=@n31))) {name=SwirlsNormals, nodeColor=<rgba 0 0 1 1>}
  NamedRerouteDeclaration(None=ComponentMask(None=@n24) {g=true}) {name=DetailMap, nodeColor=<rgba 1 0.9375 0 1>}
  n32 = NamedRerouteUsage()
  n33 = Constant() {desc=DetailMapLUTScale, r=0.5}
  n34 = QualitySwitch(Default=@n32, Epic=Add(A=@n32, B=Multiply(A=@n9, B=@n33)))
  n35 = Multiply(A=NamedRerouteUsage(), B=Constant() {desc=BaseColorDetailLUTScale, r=0.5})
  n36 = QualitySwitch(Default=@n34, Epic=LinearInterpolate(A=Add(A=@n34, B=@n35), B=Constant(), Alpha=NamedRerouteUsage()))
  n37 = CurveAtlasRowParameter(CurveTime=@n36) {
   atlas=&/CorruptionCurveAtlas
   curve=&/C_Corruption_TerrainColor
   defaultValue=4
   parameterName=TerrainLUT
  }
  n38 = CurveAtlasRowParameter(CurveTime=@n36) {
   atlas=&CorruptionCurveAtlas
   curve=&/C_Corruption_2ndStage_TerrainColor
   defaultValue=10
   parameterName=TerrainLUT_2ndStage
  }
  n39 = CollectionParameter() {
   collection=&FortniteMaterialParameters
   parameterName=Corruption2ndStage_ColorLerp
  }
  n40 = CurveAtlasRowParameter(CurveTime=NamedRerouteUsage()) {
   atlas=&CorruptionCurveAtlas
   curve=&/C_Corruption_TerrainColorDark
   defaultValue=6
   parameterName=TerrainDarkLUT
  }
  n41 = LinearInterpolate(A=LinearInterpolate(A=@n37, B=@n38, Alpha=@n39), B=@n40, Alpha=@n15)
  n42 = NamedRerouteUsage()
  NamedRerouteDeclaration(None=QualitySwitch(Default=@n41, Epic=Add(A=@n42, B=@n41))) {name=BaseColorOut, nodeColor=<rgba 0 0.7891 1 1>}
  n43 = VertexNormalWS()
  n44 = NamedRerouteUsage()
  n45 = Constant() {r=0.2}
  n46 = Normalize(VectorInput=Abs(None=LinearInterpolate(A=@n43, B=@n44, Alpha=@n45)))
  n47 = OneMinus(None=MaterialFunctionCall(
    Brightness=Constant() {r=1.25}
    Exponent=Constant() {r=4}
    Camera Vector=Constant3Vector() {constant=<rgba 0 0 1 0>}
    Vertex / Pixel Normal=QualitySwitch(Default=@n43, Epic=@n46)
   ) {materialFunction=&MF_Fresnel_FX}.Fresnel Unclamped)
  NamedRerouteDeclaration(None=FunctionInput(Preview=SmoothStep(Value=@n47)) {bUsePreviewValueAsDefault=true, inputName=CliffSlopeMask, inputType=FunctionInput_Scalar, sortPriority=2}) {name=CliffSlopeMask_In, nodeColor=<rgba 1 0.2578 0 1>}
  NamedRerouteDeclaration(None=@n28.BubblesB_UVs) {name=BubbleB_UVs, nodeColor=<rgba 1 0.9141 0 1>}
  n48 = TextureSample(UVs=NamedRerouteUsage()) {
   samplerSource=SSM_Wrap_WorldGroupSettings
   samplerType=SAMPLERTYPE_LinearColor
   texture=&T_Corruption_SwirlPack
  }
  n49 = TextureSample(UVs=NamedRerouteUsage()) {
   samplerSource=SSM_Wrap_WorldGroupSettings
   samplerType=SAMPLERTYPE_LinearColor
   texture=&T_Corruption_SwirlPack
  }
  n50 = TextureSample(UVs=NamedRerouteUsage()) {
   samplerSource=SSM_Wrap_WorldGroupSettings
   samplerType=SAMPLERTYPE_LinearColor
   texture=&T_Corruption_SwirlPack
  }
  n51 = LinearInterpolate(A=@n48.RGBA, B=Add(A=OneMinus(None=@n49.RGBA), B=Constant() {r=1.5}), Alpha=@n50.B)
  NamedRerouteDeclaration(None=QualitySwitch(Default=@n48.RGBA, High=@n51, Epic=@n51)) {name="Bubbles&&DetailsMask", nodeColor=<rgba 0 0.9297 1 1>}
  NamedRerouteDeclaration(None=@n28.BubblesLerpMask) {name=BubblesLerpMask, nodeColor=<rgba 0.7344 0 1 1>}
  n52 = NamedRerouteUsage()
  n53 = Constant() {r=0.3333}
  NamedRerouteDeclaration(None=Subtract(A=DotProduct(A=@n52, B=@n53)) {constB=0.5}) {name=BaseColorDetail, nodeColor=<rgba 0 0.8359 1 1>}
  NamedRerouteDeclaration(None=@n20.Normal) {name=NormalIn, nodeColor=<rgba 1 0 0.375 1>}
  n54 = NamedRerouteUsage()
  n55 = Constant() {r=1.35}
  n56 = Add(A=OneMinus(None=Multiply(A=@n54, B=@n55)), B=Constant() {r=-0.08}) {constB=0.11}
  n57 = Constant() {r=1.5}
  NamedRerouteDeclaration(None=Saturate(None=Power(Base=@n56, Exp=@n57) {constExponent=8})) {name=SourceColorDetailMask, nodeColor=<rgba 0.0078 1 0 1>}
  n58 = WorldPosition()
  n59 = Add(
   A=MaterialFunctionCall() {materialFunction=&ObjectPivotPoint}.Object Pivot Location
   B=VectorParameter() {parameterName=HeightFadePositionBias, defaultValue=<rgba 0 0 50 0>}.RGB
  )
  n60 = Divide(A=ComponentMask(None=Subtract(A=@n58.XYZ, B=@n59)) {b=true}, B=ScalarParameter() {defaultValue=300, parameterName=HeightFadeDivisor})
  NamedRerouteDeclaration(None=OneMinus(None=Saturate(None=@n60))) {name=HeightFade, nodeColor=<rgba 1 0.5859 0 1>}
  n61 = NamedRerouteUsage()
  n62 = Multiply(A=NamedRerouteUsage(), B=OneMinus(None=NamedRerouteUsage()))
  n63 = Subtract(A=@n16.CombinedMasks_NoiseBiased, B=NamedRerouteUsage())
  n64 = Saturate(None=QualitySwitch(Default=@n63, Epic=Subtract(A=@n63, B=@n62)) {desc=Do not use detail mask to reveal terrain beneath on non-epic platforms})
  n65 = StaticBool()
  NamedRerouteDeclaration(None=StaticSwitch(
    True=Multiply(A=@n61, B=@n64)
    False=@n64
    Value=FunctionInput(Preview=@n65) {bUsePreviewValueAsDefault=true, inputName=UseHeightFade, inputType=FunctionInput_StaticBool, sortPriority=4}
   )) {name=CorruptionLerp, nodeColor=<rgba 1 0.2 0 1>}
  NamedRerouteDeclaration(None=@n20.Metallic) {name=MetallicIn, nodeColor=<rgba 0 0.7187 1 1>}
  NamedRerouteDeclaration(None=@n20.MaterialAttributes) {name=MaterialAttributesIn, nodeColor=<rgba 0.8047 1 0 1>}
  NamedRerouteDeclaration(None=@n28.CorruptionUVs) {name=CorruptionUVs, nodeColor=<rgba 0.4297 0 1 1>}
  n66 = Add(A=Multiply(A=NamedRerouteUsage(), B=Constant3Vector() {constant=<rgba 1 1 0 0>}), B=NamedRerouteUsage())
  n67 = Constant3Vector() {desc=DarkAreaNormalIntensity, constant=<rgba 0.01 0.01 1 0>}
  NamedRerouteDeclaration(None=LinearInterpolate(A=@n66, B=Multiply(A=@n66, B=@n67), Alpha=NamedRerouteUsage())) {name=NormalOut, nodeColor=<rgba 0.1484 1 0 1>}
  NamedRerouteDeclaration(None=@n28.BubblesUVs) {name=BubbleUVs, nodeColor=<rgba 1 0.375 0 1>}
  NamedRerouteDeclaration(None=@n20.EmissiveColor) {name=EmissiveIn, nodeColor=<rgba 0.875 0 1 1>}
  n68 = Time()
  n69 = Constant3Vector() {desc=ChromaTimeScale, constant=<rgba 0.0198 0.0327 0 0>}
  n70 = ReflectionVectorWS(CustomWorldNormal=Transform(None=NamedRerouteUsage()))
  n71 = ComponentMask(None=Add(A=Multiply(A=@n68, B=@n69), B=ConstantBiasScale(None=@n70) {bias=0.5})) {g=true, r=true}
  n72 = FunctionInput(Preview=Constant() {r=1}) {bUsePreviewValueAsDefault=true, inputName=IridescenceRepeatScale, inputType=FunctionInput_Scalar, sortPriority=1}
  n73 = CurveAtlasRowParameter(CurveTime=Frac(None=DotProduct(A=@n71, B=@n72))) {
   atlas=&CorruptionCurveAtlas
   curve=&/C_Corruption_Chroma
   defaultValue=5
   parameterName=ChromaLUT
  }
  NamedRerouteDeclaration(None=Multiply(A=@n73, B=NamedRerouteUsage())) {name=Chroma, nodeColor=<rgba 0.0312 0 1 1>}
 ]}
